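# Prototype material for an Athena note in Figment_S03, named as the first note variant.
# It uses the Default Lit shading model and holds a single Texture Sample node referencing T_Athena_Note_01.
# The graph shows no parameters, math or connection from that sample to a material output.
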Material /Figment_S03/Athena/Prototype/Note/M_Athena_Note_01
graph {
 properties {pixelDepthOffsetMode=PDOM_Legacy, usedShadingModels=Default Lit}
 nodes [TextureSample() {texture=&/T_Athena_Note_01}]
}
